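; Parent consumable building actor (CBGA) that falls via projectile movement, with spawn and impact FX and sounds.
; Non-pawn hits bounce it off steep surfaces, matching classes or a designated object, with optional capped forced bounces before StopSim halts it.
; It binds to the base's death or destruction, then overlap-checks for another building or resumes falling through RepNotify bResumeSimulation.
; An Impulse event launches it after a delay, later setting replicated collision.
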(event Custom|SpawnFXSounds
  (bind _spawnfx (Variables|Gravity|GetSpawnFX))
  (bind _spawnsound (Variables|Gravity|GetSpawnSound))
  ;[1 1 1] Play Spawn FX and Sounds
  (Utilities|IsValid _spawnfx
    (:Is%20Valid
      (Effects|SpawnEmitteratLocation _spawnfx (Transformation|GetActorLocation) (Transformation|GetActorRotation)))
    (:Is%20Not%20Valid
      _))
  (Utilities|IsValid _spawnsound
    (:Is%20Valid
      (Audio|PlaySoundatLocation _spawnsound (Transformation|GetActorLocation)))))

(event EventBeginPlay
  (bind _projectilemovement (Variables|Default|GetProjectileMovement))
  ;[1 1 1] On Begin Play Check For Resim
  (CallFunction|SpawnFXSounds)
  (if (Variables|Gravity|GetResumeGravSimOnBegin)
    (Variables|Gravity|SetbResumeSimulation true)
    (Utilities|IsValid _projectilemovement
      (:Is%20Valid
        (Components|Movement|SetUpdatedComponent _projectilemovement (Variables|Default|GetStaticMesh))))
    (else
      (CallFunction|StopSim (Variables|Gravity|GetNullHit)))))

(event Custom|PlayHitFX
  (bind _hitgroundfx (Variables|Gravity|GetHitgroundFX))
  (bind _hitgroundsound (Variables|Gravity|GetHitgroundSound))
  ;[1 1 1] Handle Falling
  ;;[1 1 1] Stop Sim
  ;;;[1 1 1] Play Sounds and FX if valid
  (Utilities|IsValid _hitgroundfx
    (:Is%20Valid
      (Effects|SpawnEmitteratLocation _hitgroundfx (Transformation|GetActorLocation) (Transformation|GetActorRotation)))
    (:Is%20Not%20Valid
      _))
  (Utilities|IsValid _hitgroundsound
    (:Is%20Valid
      (Audio|PlaySoundatLocation _hitgroundsound (Transformation|GetActorLocation)))))

(event Custom|StopSim
  (params (Hit :type Hit%20Result%20Structure))
  (bind (_bblockinghit _binitialoverlap _time _distance _location _impactpoint _normal _impactnormal _physmat _hitactor) (Collision|BreakHitResult Hit))
  (bind _projectilemovement (Variables|Default|GetProjectileMovement))
  ;[1 1 1] Handle Falling
  ;;[1 1 1] Stop Sim
  (Utilities|IsValid _projectilemovement
    (:Is%20Valid
      (Game|Components|ProjectileMovement|StopSimulating _projectilemovement Hit)
      (Variables|Gravity|SetGravImpact_Loc _impactpoint)
      (Variables|Gravity|SetGravHitNormal _normal)
      (Variables|Gravity|SetbResumeSimulation false)
      (CallFunction|HandleBinding _hitactor)
      (Utilities|IsValid _hitactor
        (:Is%20Valid
          _)))
    (:Is%20Not%20Valid
      _))
  (CallFunction|PlayHitFX))

(event Custom|HandleBinding
  (params (Actor :type Actor%20Object%20Reference))
  ;[1 1 1] Handle Falling
  ;;[1 1 1] Binds Event for when Building piece dies
  (bind _asbuilding_actor (Utilities|Casting|CastToBuildingActor Actor)
    (:then
      (EventDispatchers|BindEventtoOnDied _asbuilding_actor (delegate Custom|BaseDied)))
    (:CastFailed
      (Utilities|IsValid Actor
        (:Is%20Valid
          (Game|BindEventtoOnDestroyed Actor (delegate Custom|BaseDestroyed)))))))

(event Collision|EventHit
  (params (MyComp :type Primitive%20Component%20Object%20Reference) (Other :type Actor%20Object%20Reference) (OtherComp :type Primitive%20Component%20Object%20Reference) (bSelfMoved :type Boolean) (HitLocation :type Vector :default "0, 0, 0") (HitNormal :type Vector :default "0, 0, 0") (NormalImpulse :type Vector :default "0, 0, 0") (Hit :type Hit%20Result%20Structure%20%28by%20ref%29))
  ;[1 1 1] Handle Falling
  ;;[1 1 1] When this actor hits a flat enough surface, it stops simulating movement.
  (Utilities|Casting|CastToPawn Other
    (:CastFailed
      (if (< (Math|Trig|Acos(Degrees) (Math|Vector|DotProduct HitNormal "0, 0,1.000000")) (Variables|Gravity|GetGravMaxSlope))
        (if (Variables|Gravity|GetCheckForBounce)
          (bind _hitsameclass (CallFunction|CheckforSameClasstoBounce Other))
          (if _hitsameclass
            _
            (elif (and (Variables|Gravity|GetForceBounce) (> (Variables|Gravity|GetForcedBounceMaxCount) (Variables|Gravity|GetForcedBounceCurrentCount)))
              (CallFunction|ForceBounceCBGA)
              (else
                (CallFunction|StopSim Hit)))))
        (else
          (CallFunction|BounceCBGA))))))

(event Custom|BaseDied
  (params (DamagedActor :type Actor%20Object%20Reference) (Damage :type Float%20%28single-precision%29) (InstigatedBy :type Controller%20Object%20Reference) (DamageCauser :type Actor%20Object%20Reference) (HitLocation :type Vector) (FHitComponent :type Primitive%20Component%20Object%20Reference) (BoneName :type Name) (Momentum :type Vector))
  ;[1 1 1] Handle Falling
  ;;[1 1 1] Base destroyed. Start simulating again.
  (bind (_outactors _returnvalue) (Collision|SphereOverlapActors (Transformation|GetActorLocation) 16.0 (Utilities|Array|MakeArray "ObjectTypeQuery1") "/Script/FortniteGame.BuildingActor" (Utilities|Array|MakeArray self)))
  (if _returnvalue
    (bind _output_get (Variables|Gravity|SetGravFoundBuildingOnDied _outactors))
    (bind _asbuilding_actor (Utilities|Casting|CastToBuildingActor (Utilities|Array|Get(acopy) _output_get))
      (:then
        (EventDispatchers|BindEventtoOnDied _asbuilding_actor (EventDispatchers|CreateEvent :!event Custom|BaseDied)))
      (:CastFailed
        (Game|BindEventtoOnDestroyed _asbuilding_actor (EventDispatchers|CreateEvent :!event Custom|BaseDestroyed))))
    (else
      (Variables|Gravity|SetbResumeSimulation true)
      (Networking|ForceNetUpdate))))

(event Custom|BaseDestroyed
  (params (DestroyedActor :type Actor%20Object%20Reference))
  ;[1 1 1] Handle Falling
  ;;[1 1 1] Base destroyed. Start simulating again.
  (bind (_outactors _returnvalue) (Collision|SphereOverlapActors (Transformation|GetActorLocation) 16.0 (Utilities|Array|MakeArray "ObjectTypeQuery1") "/Script/FortniteGame.BuildingActor" (Utilities|Array|MakeArray self)))
  (if _returnvalue
    (bind _output_get (Variables|Gravity|SetGravFoundBuildingOnDied _outactors))
    (bind _asbuilding_actor (Utilities|Casting|CastToBuildingActor (Utilities|Array|Get(acopy) _output_get))
      (:then
        (EventDispatchers|BindEventtoOnDied _asbuilding_actor (EventDispatchers|CreateEvent :!event Custom|BaseDied)))
      (:CastFailed
        (Game|BindEventtoOnDestroyed _asbuilding_actor (EventDispatchers|CreateEvent :!event Custom|BaseDestroyed))))
    (else
      (Variables|Gravity|SetbResumeSimulation true)
      (Networking|ForceNetUpdate))))

(event Custom|Impulse
  (params (DelayBeforeImpulse :type Float%20%28double-precision%29) (DelayBeforeCollision :type Float%20%28double-precision%29) (ImpulseOrigin :type Vector) (ImpulseAmount :type Vector) (SetCollisionAfterImpulse :type Boolean))
  ;[1 1 1] Impulse Actor
  (Utilities|FlowControl|Delay DelayBeforeImpulse)
  (Variables|/Script/Engine.MovementComponent|SetVelocity (* (Math|Vector|Normalize (- (Transformation|GetActorLocation) ImpulseOrigin)) ImpulseAmount) (Variables|Default|GetProjectileMovement))
  (Utilities|FlowControl|Delay DelayBeforeCollision)
  (Variables|Gravity|SetRepCollision SetCollisionAfterImpulse))

(event Collision|EventActorBeginOverlap
  (params (OtherActor :type Actor%20Object%20Reference))
  (|CallParentFunction OtherActor) ; This node is disabled and will not be called.\nDrag off pins to build functionality.
)

(event EventTick
  (params (DeltaSeconds :type Float%20%28single-precision%29))
  (|CallParentFunction DeltaSeconds) ; This node is disabled and will not be called.\nDrag off pins to build functionality.
)

(fn BounceCBGA
  (bind _projectilemovement (Variables|Default|GetProjectileMovement))
  ;[1 1 1] Let the CBGA Bounce and adjust its velocity.
  (Utilities|IsValid _projectilemovement
    (:Is%20Valid
      (Variables|/Script/Engine.MovementComponent|SetVelocity (Math|Vector|vector+vector (Math|Vector|MirrorVectorbyNormal (* (Variables|/Script/Engine.MovementComponent|GetVelocity _projectilemovement) (Variables|Gravity|GetBounceMult)) "0,0.000000,1.000000") (Math|Vector|MakeVector 0.0 0.0 (* (Math|Vector|VectorLength (Variables|/Script/Engine.MovementComponent|GetVelocity _projectilemovement)) (Variables|Gravity|GetBounceExtraZ)))) _projectilemovement))
    (:Is%20Not%20Valid
      _))
  (CallFunction|PlayHitFX)
  (if (Variables|Gravity|GetCountNonForceBounces)
    (Math|Integer|IncrementInt (Variables|Gravity|GetForcedBounceCurrentCount))))

(fn CheckForSameClassToBounce
  (params (HitActor :type Actor%20Object%20Reference))
  (bind _returnvalue (Utilities|GetClass HitActor))
  (bind _self self)
  (bind _returnvalue_2 (Utilities|GetClass _self))
  (bind _returnvalue_3 (Utilities|Equal(Class) _returnvalue _returnvalue_2))
  (bind _additionalbounceobject (Variables|Gravity|GetAdditionalBounceObject))
  (bind _returnvalue_4 (Utilities|Equal(Object) _additionalbounceobject HitActor))
  ;[1 1 1] Check if the hit was on an actor of the same class or on is the additional ignore actor.
  (if _returnvalue_3
    (CallFunction|BounceCBGA)
    (return _returnvalue_3)
    (else
      (Utilities|IsValid (Variables|Gravity|GetAdditionalBounceObject)
        (:Is%20Valid
          (if _returnvalue_4
            (CallFunction|BounceCBGA)
            (return _returnvalue_4)
            (else
              (return _returnvalue_4))))
        (:Is%20Not%20Valid
          (return _returnvalue_3))))))

(fn ForceBounceCBGA
  (bind _projectilemovement (Variables|Default|GetProjectileMovement))
  ;[1 1 1] Let the CBGA Bounce and adjust its velocity.
  (Utilities|IsValid _projectilemovement
    (:Is%20Valid
      (Variables|/Script/Engine.MovementComponent|SetVelocity (Math|Vector|vector+vector (Math|Vector|MirrorVectorbyNormal (* (Variables|/Script/Engine.MovementComponent|GetVelocity _projectilemovement) (Variables|Gravity|GetForcedBounceMult)) "0,0.000000,1.000000") (Math|Vector|MakeVector 0.0 0.0 (* (Math|Vector|VectorLength (Variables|/Script/Engine.MovementComponent|GetVelocity _projectilemovement)) (Variables|Gravity|GetForcedBounceExtraZ)))) _projectilemovement)
      (CallFunction|PlayHitFX)
      (Math|Integer|IncrementInt (Variables|Gravity|GetForcedBounceCurrentCount)))))

(fn OnRep_bResumeSimulation
  (bind _projectilemovement (Variables|Default|GetProjectileMovement))
  (if (Variables|Gravity|GetbResumeSimulation)
    (Utilities|IsValid _projectilemovement
      (:Is%20Valid
        (Components|Movement|SetUpdatedComponent _projectilemovement (Variables|Default|GetStaticMesh))))))

(fn OnRep_RepCollision
  (Collision|SetActorEnableCollision (Variables|Gravity|GetRepCollision)))

(fn ConstructionScript
  (|CallParentFunction))
